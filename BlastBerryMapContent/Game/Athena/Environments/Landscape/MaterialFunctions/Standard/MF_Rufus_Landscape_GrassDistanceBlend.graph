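# Landscape material function that samples a grass texture at near and far world-XY scales and blends them by distance to vary terrain grass.
# MF_TerrainDistanceFade drives the blend with a quality- and feature-level-dependent distance; UVs can optionally use vertex interpolators.
# A static switch enables per-channel power-curve remapping of blended and distant RGB into narrow ranges.
# It also outputs grass scale and alpha, which uses only the near sample under virtual texturing.
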MaterialFunction /BlastBerryMapContent/Game/Athena/Environments/Landscape/MaterialFunctions/Standard/MF_Rufus_Landscape_GrassDistanceBlend
graph {nodes [
  n1 = ComponentMask(None=WorldPosition() {worldPositionShaderOffset=WPT_ExcludeAllShaderOffsets}.XYZ) {g=true, r=true}
  n2 = FunctionOutput(None=Divide(A=@n1, B=ScalarParameter() {defaultValue=700, parameterName=GrassNearScale})) {outputName=GrassScale}
  n3 = FunctionInput(Preview=StaticBool()) {bUsePreviewValueAsDefault=true, inputName=UseVertexInterp, inputType=FunctionInput_StaticBool}
  n4 = TextureSampleParameter2D(UVs=StaticSwitch(True=VertexInterpolator(VS=@n2).PS, False=@n2, Value=@n3)) {
   group=Grass
   parameterName=Grass
   samplerSource=SSM_Wrap_WorldGroupSettings
   texture=/Game/Athena/Environments/Landscape/Textures/T_GrasslandsChannelNormalized
  }
  n5 = Divide(A=@n1, B=Constant() {r=3000})
  n6 = TextureSampleParameter2D(UVs=StaticSwitch(True=VertexInterpolator(VS=@n5).PS, False=@n5, Value=@n3)) {
   group=Grass
   parameterName=Grass
   samplerSource=SSM_Wrap_WorldGroupSettings
   texture=&T_GrasslandsChannelNormalized
  }
  n7 = Constant() {r=55000}
  n8 = Constant() {r=45000}
  n9 = QualitySwitch(Default=@n7, Low=@n8, High=@n7, Medium=@n7)
  n10 = MaterialFunctionCall!(Distance=FeatureLevelSwitch(Default=@n9, ES2=@n8, ES3_1=@n8, SM4_REMOVED=@n9, SM5=@n9, SM6=@n9)) {materialFunction=/Game/Athena/Environments/Landscape/MaterialFunctions/MF_TerrainDistanceFade}
  n11 = LinearInterpolate(A=@n4.RGB, B=@n6.RGB, Alpha=@n10.Result)
  n12 = MaterialFunctionCall(Float3=@n11) {materialFunction=&BreakOutFloat3Components}
  n13 = OneMinus(None=@n12.R)
  n14 = LinearInterpolate(Alpha=OneMinus(None=Power(Base=@n13) {constExponent=2.2})) {constA=0.04, constB=0.11}
  n15 = OneMinus(None=@n12.G)
  n16 = LinearInterpolate(Alpha=OneMinus(None=Power(Base=@n15) {constExponent=2.2})) {constA=0.12, constB=0.2}
  n17 = OneMinus(None=@n12.B)
  n18 = LinearInterpolate(Alpha=OneMinus(None=Power(Base=@n17))) {constA=0.015, constB=0.05}
  FunctionOutput(None=StaticSwitchParameter(
    True=MaterialFunctionCall(X=@n14, Y=@n16, Z=@n18) {materialFunction=&MakeFloat3}.Result
    False=@n11
   ) {defaultValue=true, parameterName=Use Remap Grass}) {outputName=Grass RGB}
  FunctionOutput(None=VirtualTextureFeatureSwitch!(No=LinearInterpolate(A=@n4.A, B=@n6.A, Alpha=@n10.Result), Yes=@n4.A)) {outputName=Grass A, sortPriority=1}
  n19 = MaterialFunctionCall(Float3=@n6.RGB) {materialFunction=&BreakOutFloat3Components}
  n20 = OneMinus(None=@n19.R)
  n21 = LinearInterpolate(Alpha=OneMinus(None=Power(Base=@n20) {constExponent=2.2})) {constA=0.04, constB=0.11}
  n22 = OneMinus(None=@n19.G)
  n23 = LinearInterpolate(Alpha=OneMinus(None=Power(Base=@n22) {constExponent=2.2})) {constA=0.12, constB=0.2}
  n24 = OneMinus(None=@n19.B)
  n25 = LinearInterpolate(Alpha=OneMinus(None=Power(Base=@n24))) {constA=0.015, constB=0.05}
  FunctionOutput(None=StaticSwitchParameter(
    True=MaterialFunctionCall(X=@n21, Y=@n23, Z=@n25) {materialFunction=&MakeFloat3}.Result
    False=@n6.RGB
   ) {defaultValue=true, parameterName=Use Remap Grass}) {outputName=Distant Grass RGB, sortPriority=3}
  TextureSample() {texture=/Game/Athena/Environments/Landscape/Textures/T_Athena_Grass_Farm_ColorMatched_D_2}
 ]}
