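# Landscape material function that builds the forest floor layer for Athena terrain from a UV input, returning material attributes plus a separate Height output.
# The Forest texture parameter supplies base color, its green channel drives roughness between tunable bounds, and its alpha serves as height, specular blend and opacity mask.
# A fixed normal map is scaled by a vector parameter.
# An extra forest floor diffuse sample is left unconnected.
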MaterialFunction /BlastBerryMapContent/Game/Athena/Environments/Landscape/MaterialFunctions/Standard/MF_Athena_Forest_01
graph {
 properties {bExposeToLibrary=true}
 nodes [
  n1 = FunctionInput(Preview=TextureCoordinate()) {bUsePreviewValueAsDefault=true, inputName=UVs, inputType=FunctionInput_Vector2}
  n2 = TextureSampleParameter2D(UVs=@n1) {
   group=Forest
   parameterName=Forest
   samplerSource=SSM_Wrap_WorldGroupSettings
   texture=/Game/Packages/DS_Fortnight/SM/Textures/T_FORT_ForestFloor01
  }
  n3 = ScalarParameter() {defaultValue=0.3, group=Forest, parameterName=Forest_SpecMin}
  n4 = ScalarParameter() {defaultValue=0.6, group=Forest, parameterName=Forest_SpecMax}
  n5 = FunctionOutput(None=@n2.A) {outputName=Height, sortPriority=1}
  n6 = ScalarParameter() {defaultValue=1, group=Forest, parameterName=Forest_RoughMin}
  n7 = ScalarParameter() {defaultValue=-1, group=Forest, parameterName=Forest_RoughMax}
  n8 = SetMaterialAttributes!(
   MaterialAttributes
   BaseColor=@n2.RGB
   Specular=Saturate(None=LinearInterpolate(A=@n3, B=@n4, Alpha=@n5))
   Roughness=Saturate(None=LinearInterpolate(A=@n6, B=@n7, Alpha=@n2.G))
   Normal=Multiply(
    A=TextureSample(UVs=@n1) {samplerType=SAMPLERTYPE_Normal, texture=/Game/Environments/Sets/JungleTemple/Textures/Temple_Alt_gray/T_FORT_ForestFloor01_N}.RGB
    B=VectorParameter() {group=Forest, parameterName=ForestFloor_NormalAmount, defaultValue=<rgba 3 3 1 1>}.RGB
   )
   OpacityMask=@n5
  ) {attributeSetTypes [
    #1
    #2
    #3
    #4
    #5
   ]}
  FunctionOutput(None=@n8)
  TextureSample() {texture=/Game/Athena/Environments/Landscape/Textures/T_Athena_ForestFloor_D}
 ]
}
