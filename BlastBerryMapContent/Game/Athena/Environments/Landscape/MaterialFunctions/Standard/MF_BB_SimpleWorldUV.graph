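# Material function for the BlastBerry map landscape materials that produces simple planar UVs from world position.
# It divides world position XY by the negated absolute value of the Vector2 Scale input, so the sign of Scale is ignored and both axes are always flipped.
# Scale sets the world-space tile size per axis and uses its preview value as the default.
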MaterialFunction /BlastBerryMapContent/Game/Athena/Environments/Landscape/MaterialFunctions/Standard/MF_BB_SimpleWorldUV
graph {nodes [
  n1 = Abs(None=FunctionInput(Preview=Constant2Vector() {g=2000, r=2000}) {bUsePreviewValueAsDefault=true, inputName=Scale, inputType=FunctionInput_Vector2})
  FunctionOutput(None=Divide(A=WorldPosition().XY, B=Multiply(A=@n1) {constB=-1}))
 ]}
