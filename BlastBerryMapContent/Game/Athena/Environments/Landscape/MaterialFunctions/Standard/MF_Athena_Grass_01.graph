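# Landscape material function for the BlastBerry map's grass layer, outputting material attributes plus plain base color, height and UVs.
# Scaled world UVs sample a channel-packed grass texture: RGB optionally remaps into narrow color ranges, green drives specular and roughness, alpha gives height.
# World-position masks from MF_GrassMicroColorMasks support macro color tinting, followed by a blend toward GrassLerpColor.
# A grasslands normal map passes through a world-aligned normal function with adjustable strength.
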MaterialFunction /BlastBerryMapContent/Game/Athena/Environments/Landscape/MaterialFunctions/Standard/MF_Athena_Grass_01
graph {
 properties {bExposeToLibrary=true}
 nodes [
  Constant() {r=1}
  Constant() {r=700}
  Constant() {r=0.2}
  Constant() {r=0.7}
  Constant() {r=1}
  Constant() {r=0.75}
  Constant()
  n1 = TextureSample(UVs=NamedRerouteUsage()) {
   samplerSource=SSM_Wrap_WorldGroupSettings
   samplerType=SAMPLERTYPE_Normal
   texture=/Game/Packages/DS_Fortnite_Terrain_NoLOD/Textures/T_Grasslands_N
  }
  n2 = SetMaterialAttributes!(
   MaterialAttributes
   BaseColor=NamedRerouteUsage()
   Specular=NamedRerouteUsage()
   Roughness=NamedRerouteUsage()
   Normal=MaterialFunctionCall!(
    Normals=@n1.RGB
    TransformToTangentSpace=StaticBool() {value=true}
    Normal Strength=ScalarParameter() {defaultValue=0.6, group=Grass, parameterName=GrassNormalAmount}
   ) {materialFunction=/Game/Athena/Apollo/Environments/Landscape/Materials/Biome/CommonMaterials/MF_HQ_WorldAlignedNormal}.Result
   OpacityMask=NamedRerouteUsage()
  ) {attributeSetTypes [
    #1
    #2
    #3
    #4
    #5
   ]}
  FunctionOutput(None=@n2)
  FunctionOutput(None=NamedRerouteUsage()) {outputName=BaseColor_Plain}
  n3 = MaterialFunctionCall!(Scale=ScalarParameter() {defaultValue=700, parameterName=GrassNearScale}) {materialFunction=&/MF_BB_SimpleWorldUV}
  n4 = TextureSampleParameter2D(UVs=FunctionOutput(None=NamedRerouteDeclaration(None=@n3.Result) {name=GrassScale, nodeColor=<rgba 1 0 0.8203 1>}) {outputName=GrassScale}) {
   group=Grass
   parameterName=Grass
   samplerSource=SSM_Wrap_WorldGroupSettings
   texture=/Game/Athena/Environments/Landscape/Textures/T_GrasslandsChannelNormalized
  }
  FunctionOutput(None=NamedRerouteDeclaration(None=@n4.A) {name=GrassHeight, nodeColor=<rgba 1 0 0.8203 1>}) {outputName=Height}
  LinearInterpolate(
   A=VectorParameter() {group=Grass, parameterName=SSC-1, defaultValue=<rgba 1 1 1 1>}.RGB
   B=VectorParameter() {group=Grass, parameterName=SSC-2, defaultValue=<rgba 1 1 1 1>}.RGB
   Alpha=NamedRerouteUsage()
  )
  n5 = ComponentMask(None=VectorParameter() {group=Grass, parameterName=GrassMacroColorA, defaultValue=<rgba 0.4971 0.9792 0 1>}.RGB) {b=true, g=true, r=true}
  n6 = LinearInterpolate(
   A=Constant3Vector() {constant=<rgba 1 1 1 0>}
   B=ComponentMask(None=VectorParameter() {group=Grass, parameterName=SSC-3, defaultValue=<rgba 0.9479 0.9048 0.077 1>}.RGB) {b=true, g=true, r=true}
   Alpha=NamedRerouteUsage()
  )
  n7 = Saturate(None=Multiply(A=NamedRerouteUsage(), B=Constant() {r=4}))
  n8 = NamedRerouteUsage()
  n9 = LinearInterpolate(A=Multiply(A=LinearInterpolate(A=@n5, B=@n6, Alpha=@n7), B=@n8), B=@n8, Alpha=NamedRerouteUsage())
  n10 = VectorParameter() {group=Grass, parameterName=GrassLerpColor, channelNames {a=LerpColorAmount, b="", g="", r=""}, defaultValue=<rgba 0.0362 0.13 0.0026 0.6>}
  NamedRerouteDeclaration(None=LinearInterpolate(A=Saturate(None=@n9), B=@n10, Alpha=Saturate(None=@n10.LerpColorAmount))) {name=Final_Color, nodeColor=<rgba 0 1 0.1094 1>}
  n11 = OneMinus(None=@n4.R)
  n12 = LinearInterpolate(Alpha=OneMinus(None=Power(Base=@n11) {constExponent=2.2})) {constA=0.04, constB=0.11}
  n13 = OneMinus(None=@n4.G)
  n14 = LinearInterpolate(Alpha=OneMinus(None=Power(Base=@n13) {constExponent=2.2})) {constA=0.12, constB=0.2}
  n15 = OneMinus(None=@n4.B)
  n16 = LinearInterpolate(Alpha=OneMinus(None=Power(Base=@n15))) {constA=0.015, constB=0.05}
  NamedRerouteDeclaration(None=StaticSwitchParameter(
    True=MaterialFunctionCall(X=@n12, Y=@n14, Z=@n16) {materialFunction=&MakeFloat3}.Result
    False=@n4.RGB
   ) {defaultValue=true, parameterName=Use Remap Grass}) {name=Base Color, nodeColor=<rgba 0.3594 0 1 1>}
  n17 = MaterialFunctionCall!(SamplingPosition=WorldPosition().XYZ) {materialFunction=/Game/Athena/Asteria/Environments/Landscape/Materials/MaterialFunctions/MF_GrassMicroColorMasks}
  NamedRerouteDeclaration(None=@n17.GrassColorMask3) {name=GrassColorMask3, nodeColor=<rgba 1 0 0.8203 1>}
  NamedRerouteDeclaration(None=@n17.GrassColorMask4) {name=GrassColorMask4, nodeColor=<rgba 1 0 0.8203 1>}
  NamedRerouteDeclaration(None=@n17.GrassMacroColorMask) {name=GrassMacroColorMask, nodeColor=<rgba 1 0 0.8203 1>}
  NamedRerouteDeclaration(None=@n17.GrassColorMask2) {name=GrassColorMask2, nodeColor=<rgba 1 0 0.8203 1>}
  n18 = ScalarParameter() {defaultValue=0.3, group=Grass, parameterName=Grass_SpecLow}
  n19 = ScalarParameter() {defaultValue=0.5, group=Grass, parameterName=Grass_SpecHigh}
  NamedRerouteDeclaration(None=Saturate(None=LinearInterpolate(A=@n18, B=@n19, Alpha=@n4.G))) {name=Spec, nodeColor=<rgba 0.3594 0 1 1>}
  n20 = Saturate(None=LinearInterpolate(
    A=ScalarParameter() {defaultValue=1, group=Grass, parameterName=Grass_RoughLow}
    B=ScalarParameter() {defaultValue=0.75, group=Grass, parameterName=Grass_RoughHigh}
    Alpha=@n4.G
   ))
  n21 = ScalarParameter() {defaultValue=0.05, group=Grass, parameterName=Grass_RoughAdd}
  NamedRerouteDeclaration(None=Saturate(None=Add(A=@n20, B=@n21))) {name=Rough, nodeColor=<rgba 0.3594 0 1 1>}
  NamedRerouteDeclaration(None=@n17.GrassColorMask1) {name=GrassColorMask1, nodeColor=<rgba 1 0 0.8203 1>}
 ]
}
