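# Landscape material function that builds the gravel layer for Athena terrain, returning material attributes plus plain base color and height outputs.
# One gravel texture drives the layer: RGB is tinted by the landscape parameter collection's GravelTint and a fixed warm color.
# Its red channel remaps to specular and roughness through Gravel-group scalar parameters, while alpha supplies height, also written to OpacityMask.
# A fixed rock normal map shares the input UVs.
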MaterialFunction /BlastBerryMapContent/Game/Athena/Environments/Landscape/MaterialFunctions/Standard/MF_Athena_Gravel_01
graph {
 properties {bExposeToLibrary=true}
 nodes [
  n1 = ComponentMask(None=CollectionParameter() {collection=/BlastBerryMapContent/Game/Athena/Environments/Landscape/MPC/MPC_Landscape, parameterName=GravelTint}) {b=true, g=true, r=true}
  n2 = FunctionInput(Preview=TextureCoordinate()) {bUsePreviewValueAsDefault=true, inputType=FunctionInput_Vector2}
  n3 = TextureSampleParameter2D(UVs=@n2) {group=Gravel, parameterName=Gravel, samplerSource=SSM_Wrap_WorldGroupSettings, texture=/Game/Packages/DS_Fortnight/WhiteRockCustom_D}
  n4 = FunctionOutput(None=@n3.RGB) {outputName=Plain_BaseColor, sortPriority=2}
  n5 = ScalarParameter() {defaultValue=0.2, group=Gravel, parameterName=Gravel_SpecMin}
  n6 = ScalarParameter() {defaultValue=1.25, group=Gravel, parameterName=Gravel_SpecMax}
  n7 = Saturate(None=LinearInterpolate(
    A=ScalarParameter() {defaultValue=1, group=Gravel, parameterName=Gravel_RoughMin}
    B=ScalarParameter() {defaultValue=-1, group=Gravel, parameterName=Gravel_RoughMax}
    Alpha=@n3.R
   ))
  n8 = ScalarParameter() {defaultValue=0.5, group=Gravel, parameterName=Gravel_RoughAdd}
  n9 = SetMaterialAttributes!(
   MaterialAttributes
   BaseColor=Multiply(A=Constant3Vector() {constant=<rgba 1 0.8173 0.526 1>}, B=Multiply(A=@n1, B=@n4))
   Specular=Saturate(None=LinearInterpolate(A=@n5, B=@n6, Alpha=@n3.R))
   Roughness=Saturate(None=Add(A=@n7, B=@n8))
   Normal=TextureSample(UVs=@n2) {samplerType=SAMPLERTYPE_Normal, texture=/Game/Packages/DS_Fortnight/WhiteRockCustom_N}.RGB
   OpacityMask=FunctionOutput(None=@n3.A) {outputName=Height, sortPriority=1}
  ) {attributeSetTypes [
    #1
    #2
    #3
    #4
    #5
   ]}
  FunctionOutput(None=@n9)
 ]
}
