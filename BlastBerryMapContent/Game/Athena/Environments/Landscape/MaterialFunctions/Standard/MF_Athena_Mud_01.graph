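# Landscape material function that builds a tinted mud layer for the BlastBerry map and returns it as material attributes.
# One mud texture supplies base color, tinted by MudTint from the landscape parameter collection, with untinted color also exposed as BaseColor_Plain.
# Its red channel blends between specular bounds, while alpha blends between roughness bounds and feeds opacity mask.
# A fixed Asteria mud normal map shares the optional UV input.
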MaterialFunction /BlastBerryMapContent/Game/Athena/Environments/Landscape/MaterialFunctions/Standard/MF_Athena_Mud_01
graph {
 properties {bExposeToLibrary=true}
 nodes [
  Constant() {r=1}
  Constant() {r=0.5}
  Constant() {r=0.2}
  n1 = CollectionParameter() {collection=/BlastBerryMapContent/Game/Athena/Environments/Landscape/MPC/MPC_Landscape, parameterName=MudTint}
  n2 = FunctionInput(Preview=TextureCoordinate()) {bUsePreviewValueAsDefault=true, inputType=FunctionInput_Vector2}
  n3 = TextureSampleParameter2D(UVs=@n2) {
   group=Mud
   parameterName=Mud
   samplerSource=SSM_Wrap_WorldGroupSettings
   texture=/Game/Athena/Environments/Landscape/Textures/Arid/T_Athena_Arid_MudVariant_01
  }
  n4 = ScalarParameter() {defaultValue=0.2, group=Mud, parameterName=Mud_SpecMin}
  n5 = ScalarParameter() {defaultValue=0.8, group=Mud, parameterName=Mud_SpecMax}
  n6 = ScalarParameter() {defaultValue=0.9, group=Mud, parameterName=Mud_RoughMin}
  n7 = ScalarParameter() {defaultValue=1, group=Mud, parameterName=Mud_RoughMax}
  n8 = SetMaterialAttributes!(
   MaterialAttributes
   BaseColor=Multiply(A=ComponentMask(None=@n1) {b=true, g=true, r=true}, B=FunctionOutput(None=@n3.RGB) {outputName=BaseColor_Plain})
   Specular=Saturate(None=LinearInterpolate(A=@n4, B=@n5, Alpha=@n3.R))
   Roughness=Saturate(None=LinearInterpolate(A=@n6, B=@n7, Alpha=@n3.A))
   Normal=TextureSample(UVs=@n2) {samplerType=SAMPLERTYPE_Normal, texture=/Game/Environments/Asteria/Terrain/Textures/T_Asteria_Mud_01_N}.RGB
   OpacityMask=@n3.A
  ) {attributeSetTypes [
    #1
    #2
    #3
    #4
    #5
   ]}
  FunctionOutput(None=@n8)
 ]
}
