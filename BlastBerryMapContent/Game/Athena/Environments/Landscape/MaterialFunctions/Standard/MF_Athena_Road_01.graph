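# Material function for the Athena landscape that outputs road surface material attributes from street road diffuse, mask and normal textures sampled at input UVs.
# Diffuse supplies base color and opacity mask; mask channels are remapped into saturated specular and roughness.
# A UseSimple static switch instead derives specular from diffuse red, with constant roughness and a flat normal.
# Road brightness, specular and roughness parameters are unconnected; hardcoded constants drive the remaps.
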MaterialFunction /BlastBerryMapContent/Game/Athena/Environments/Landscape/MaterialFunctions/Standard/MF_Athena_Road_01
graph {
 properties {bExposeToLibrary=true}
 nodes [
  n1 = FunctionInput(Preview=TextureCoordinate()) {bUsePreviewValueAsDefault=true, inputType=FunctionInput_Vector2}
  n2 = TextureSample(UVs=@n1) {samplerSource=SSM_Wrap_WorldGroupSettings, texture=/Game/Packages/DS_Fortnight/SM/Textures/T_Street_Road_D}
  n3 = LinearInterpolate(A=Constant() {r=-0.35}, B=Constant() {r=4}, Alpha=@n2.R)
  n4 = TextureSample(UVs=@n1) {samplerType=SAMPLERTYPE_Masks, texture=/Game/Packages/DS_Fortnight/SM/Textures/T_Street_Road_M}
  n5 = LinearInterpolate(A=Constant() {r=0.2}, B=Constant() {r=0.5}, Alpha=@n4.R)
  n6 = FunctionInput(Preview=StaticBool()) {bUsePreviewValueAsDefault=true, inputName=UseSimple, inputType=FunctionInput_StaticBool}
  n7 = Constant() {r=1}
  n8 = LinearInterpolate(A=Constant() {r=-0.2}, B=Constant() {r=1}, Alpha=@n4.B)
  n9 = SetMaterialAttributes!(
   MaterialAttributes
   BaseColor=@n2.RGB
   Specular=Saturate(None=StaticSwitch(True=@n3, False=@n5, Value=@n6))
   Roughness=Saturate(None=StaticSwitch(True=@n7, False=@n8, Value=@n6))
   Normal=StaticSwitch(
    True=Constant3Vector() {constant=<rgba 0 0 1 0>}
    False=TextureSample(UVs=@n1) {samplerType=SAMPLERTYPE_Normal, texture=/Game/Packages/DS_Fortnight/SM/Textures/T_Street_Road_N}.RGB
    Value=@n6
   )
   OpacityMask=@n2.A
  ) {attributeSetTypes [
    #1
    #2
    #3
    #4
    #5
   ]}
  FunctionOutput(None=@n9)
  Multiply(A=@n2.RGB, B=ScalarParameter() {defaultValue=1, group=Road, parameterName=Road_Brightness})
  ScalarParameter() {defaultValue=-0.35, group=Road, parameterName=Road_Spec_2}
  ScalarParameter() {defaultValue=4, group=Road, parameterName=Road_Spec_3}
  ScalarParameter() {defaultValue=0.2, group=Road, parameterName=Road_Spec_Low}
  ScalarParameter() {defaultValue=0.5, group=Road, parameterName=Road_Spec_High}
  ScalarParameter() {defaultValue=-0.2, group=Road, parameterName=Road_Rough_Min}
  ScalarParameter() {defaultValue=1, group=Road, parameterName=Road_Rough_Max}
 ]
}
