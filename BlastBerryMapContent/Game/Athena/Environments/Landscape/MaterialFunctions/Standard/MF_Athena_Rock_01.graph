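# Landscape rock material function for the BlastBerry map, outputting material attributes, plain base color and a height-graded brightness factor from MPC_Landscape.
# World-aligned cave-wall and white-rock projections blend with a larger-scale wall projection by vertex normal Z.
# The color's red channel drives specular and roughness.
# Low quality and mobile use one larger-scale normal and a single color sample, choosing XZ or YZ with explicit derivatives.
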MaterialFunction /BlastBerryMapContent/Game/Athena/Environments/Landscape/MaterialFunctions/Standard/MF_Athena_Rock_01
graph {
 properties {bExposeToLibrary=true}
 nodes [
  Constant() {r=-4}
  Constant() {r=5}
  Constant() {r=0.8}
  Constant() {r=0.725}
  Constant() {r=0.725}
  Constant() {r=-0.5}
  Constant() {r=1.5}
  n1 = ScalarParameter() {defaultValue=1535, group=Rock, parameterName=TexSize}
  n2 = Multiply(A=VertexNormalWS(), B=ScalarParameter() {defaultValue=0.62, group=Rock, parameterName=RockProjNormalMP_Z})
  n3 = ScalarParameter() {defaultValue=10, group=Rock, parameterName=RockProjContrast}
  n4 = MaterialFunctionCall!(
   TextureObject=TextureObject() {texture=/Game/Packages/DS_Fortnite_Terrain_NoLOD/Textures/T_Cave_Exterior_Wall_D}
   TextureSize=@n1
   WorldPosition
   Export Float 4=StaticBool() {value=true}
   World Space Normal_Z=@n2
   TextureObject_Z=TextureObject() {texture=/Game/Packages/DS_Fortnight/WhiteRockCustom_D}
   World Space Normal=VertexNormalWS()
   ProjectionTransitionContrast=@n3
  ) {materialFunction=/Game/Athena/Prototype/Terrain/WorldAlignedTexture_ZTex}
  n5 = Divide(A=@n4.XYZ Texture)
  n6 = ScalarParameter() {defaultValue=3741, group=Rock, parameterName=TexSize_2}
  n7 = MaterialFunctionCall(
   TextureObject=TextureObject() {texture=&T_Cave_Exterior_Wall_D}
   TextureSize=@n6
   Export Float 4=StaticBool() {value=true}
   World Space Normal=VertexNormalWS()
   ProjectionTransitionContrast=@n3
  ) {materialFunction=&WorldAlignedTexture_ZTex}
  n8 = Divide(A=@n7.XY Texture)
  n9 = ScalarParameter() {defaultValue=5, group=Rock, parameterName=TransitonRockWall_LOW}
  n10 = ScalarParameter() {defaultValue=-4, group=Rock, parameterName=TransitonRockWall_High}
  n11 = ComponentMask(None=VertexNormalWS()) {b=true}
  n12 = Saturate(None=Add(A=LinearInterpolate(A=@n9, B=@n10, Alpha=@n11), B=ScalarParameter() {defaultValue=0.8, group=Rock, parameterName=TransitonRockWall_Offset}))
  n13 = ShadingPathSwitch(
   Default=QualitySwitch(Default=LinearInterpolate(A=@n5, B=@n8, Alpha=@n12), Low=NamedRerouteUsage()) {desc=Use only one rock blend on Medium and Low}
   Mobile=NamedRerouteUsage()
  )
  n14 = FunctionOutput(None=ComponentMask(None=@n13) {b=true, g=true, r=true}) {outputName=BaseColor_Plain}
  n15 = ScalarParameter() {defaultValue=0.25, group=Rock, parameterName=Rock_SpecMin}
  n16 = ScalarParameter() {defaultValue=0.75, group=Rock, parameterName=Rock_SpecMax}
  n17 = ComponentMask(None=@n13) {r=true}
  n18 = ScalarParameter() {defaultValue=1, group=Rock, parameterName=Rock_RoughMin}
  n19 = ScalarParameter() {defaultValue=-0.5, group=Rock, parameterName=Rock_RoughMax}
  n20 = ScalarParameter() {defaultValue=-10, group=Rock, parameterName=RockProjContrast_Normal}
  n21 = MaterialFunctionCall!(
   TextureObject=TextureObject() {samplerType=SAMPLERTYPE_Normal, texture=/Game/Packages/DS_Fortnite_Terrain_NoLOD/Textures/T_Cave_Exterior_Wall_N}
   TextureSize=@n1
   WorldSpace
   Normal
   ProjectionTransitionContrast=@n20
   WorldPosition
   Use High Quality Normals
  ) {materialFunction=/Game/Athena/Environments/Materials/WorldAlignedNormalContrast}
  n22 = MaterialFunctionCall(TextureObject=TextureObject() {samplerType=SAMPLERTYPE_Normal, texture=/Game/Packages/DS_Fortnight/WhiteRockCustom_N}, TextureSize=@n1) {materialFunction=&WorldAlignedNormalContrast}
  n23 = ComponentMask(None=@n2) {b=true}
  n24 = MaterialFunctionCall(
   TextureObject=TextureObject() {samplerType=SAMPLERTYPE_Normal, texture=&T_Cave_Exterior_Wall_N}
   TextureSize=@n6
   ProjectionTransitionContrast=@n20
  ) {materialFunction=&WorldAlignedNormalContrast}
  n25 = LinearInterpolate(
   A=LinearInterpolate(
    A=@n21.XY Texture
    B=@n22.Z Texture
    Alpha=MaterialFunctionCall(In=@n23, Contrast=@n3) {materialFunction=&CheapContrast}.Result
   )
   B=@n24.XY Texture
   Alpha=@n12
  )
  n26 = SetMaterialAttributes!(
   MaterialAttributes
   BaseColor=@n14
   Specular=Saturate(None=LinearInterpolate(A=@n15, B=@n16, Alpha=@n17))
   Roughness=Saturate(None=LinearInterpolate(A=@n18, B=@n19, Alpha=@n17))
   Normal=ShadingPathSwitch(Default=QualitySwitch(Default=@n25, Low=@n24.XY Texture) {desc=Use only one rock blend on Medium and Low}, Mobile=@n24.XY Texture)
   OpacityMask=ComponentMask(None=@n13) {a=true}
  ) {attributeSetTypes [
    #1
    #2
    #3
    #4
    #5
   ]}
  FunctionOutput(None=@n26)
  n27 = Multiply(A=@n14, B=CollectionParameter() {collection=/Game/Athena/Environments/Landscape/MPC/MPC_Landscape, parameterName=Rock_Bottom_Brightness})
  n28 = Multiply(A=@n14, B=CollectionParameter() {collection=&MPC_Landscape, parameterName=Rock_Top_Brightness})
  n29 = MaterialFunctionCall!(height_Offset, height_GradScale, Coords=WorldPosition().XYZ) {materialFunction=/Game/Athena/Environments/Landscape/MaterialFunctions/MF_Athena_WorldHeightGrad}
  n30 = ScalarParameter() {group=Rock, parameterName=HMP}
  n31 = ScalarParameter() {group=Rock, parameterName=Hblend_Contrast}
  FunctionOutput(None=LinearInterpolate(
    A=CollectionParameter() {collection=&MPC_Landscape, parameterName=Rock_Bottom_Brightness}
    B=CollectionParameter() {collection=&MPC_Landscape, parameterName=Rock_Top_Brightness}
    Alpha=MaterialFunctionCall(A=@n27, B=@n28, Transition Phase=@n29.Result, Height Texture=@n30, Contrast=@n31) {materialFunction=&HeightLerp}.Alpha
   )) {outputName=Rock_Color}
  FunctionOutput(None=MakeMaterialAttributes!(
    BaseColor=@n7.XY Texture
    Metallic
    Specular
    Roughness
    Anisotropy
    EmissiveColor
    Opacity
    OpacityMask
    Normal=@n24.XY Texture
    Tangent
    WorldPositionOffset
    SubsurfaceColor
    ClearCoat
    ClearCoatRoughness
    AmbientOcclusion
    Refraction
    CustomizedUVs_0
    CustomizedUVs_1
    CustomizedUVs_2
    CustomizedUVs_3
    CustomizedUVs_4
    CustomizedUVs_5
    CustomizedUVs_6
    CustomizedUVs_7
    PixelDepthOffset
    ShadingModel
    Displacement
   )) {outputName=Rock_02_XY_Mat}
  FunctionOutput(None=@n3) {outputName=Proj_Contrast}
  n32 = Abs(None=ComponentMask(None=@n2) {b=true})
  LinearInterpolate(
   A=NamedRerouteUsage()
   B=Divide(A=@n4.Z Texture)
   Alpha=ComponentMask(None=MaterialFunctionCall(In=@n32, Contrast=@n3) {materialFunction=&CheapContrast}.Result) {r=true}
  )
  Multiply(A=VertexNormalWS(), B=ScalarParameter() {defaultValue=0.725, group=Rock, parameterName=RockProjNormalMP_Z_1})
  Multiply(A=@n22.Z Texture)
  n33 = AppendVector(
   A=ScalarParameter() {defaultValue=19.716, group=Rock, parameterName=Rock_Tiling_X_1}
   B=ScalarParameter() {defaultValue=43.9378, group=Rock, parameterName=Rock_Tiling_Y_1}
  )
  n34 = Divide(A=LandscapeLayerCoords() {mappingScale=1, mappingType=TCMT_XZ}, B=@n33)
  n35 = Divide(A=LandscapeLayerCoords() {mappingScale=1, mappingType=TCMT_YZ}, B=@n33)
  n36 = Abs(None=VertexNormalWS())
  n37 = Ceil(None=Subtract(A=ComponentMask(None=@n36) {r=true}, B=ComponentMask(None=@n36) {g=true}))
  n38 = DDX(Value=@n34)
  n39 = DDY(Value=@n34)
  n40 = DDX(Value=@n35)
  n41 = DDY(Value=@n35)
  n42 = Max(A=Abs(None=AppendVector(A=@n38, B=@n39)), B=Abs(None=AppendVector(A=@n40, B=@n41)))
  n43 = TextureSample!(
   UVs=LinearInterpolate(A=@n34, B=@n35, Alpha=Saturate(None=@n37))
   Tex
   "DDX(UVs)"=ComponentMask(None=@n42) {g=true, r=true}
   "DDY(UVs)"=ComponentMask(None=@n42) {a=true, b=true}
   Apply View MipBias
  ) {
   mipValueMode=TMVM_Derivative
   samplerSource=SSM_Wrap_WorldGroupSettings
   texture=&T_Cave_Exterior_Wall_D
  }
  NamedRerouteDeclaration(None=Divide(A=AppendVector(A=@n43.RGB, B=@n43.RGB))) {name=Low/MobileRock, nodeColor=<rgba 0.0078 0 1 1>}
  OneMinus(None=MaterialFunctionCall!(
    Diffuse 1
    Diffuse 2
    Transistion Phase=@n12
    Height Texture 1=ComponentMask(None=@n4.XY Texture) {a=true}
    Height Texture 2=ComponentMask(None=@n7.XY Texture) {a=true}
    Contrast=ScalarParameter() {parameterName=erty47457578}
   ) {materialFunction=/Engine/Functions/Engine_MaterialFunctions02/Texturing/HeightLerpWithTwoHeightMaps}.Alpha)
  Reroute(None=@n21.XY Texture)
  ScalarParameter() {parameterName=ghdfghfghd56546}
  TextureObject() {texture=&WhiteRockCustom_D}
 ]
}
